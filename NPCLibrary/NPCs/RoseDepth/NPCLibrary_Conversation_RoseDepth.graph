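# Conversation database for the RoseDepth NPC, The Night Rose, defining the greeting and top-level branches of talking to that character.
# One speech node picks between a default line and contextual lines gated by first-ever conversation, per-match conversation count or the player's character cosmetic, with priorities set.
# The count checks use the IronLilac NPC tag.
# Link nodes then branch to quests, basic services, RoseDepth dialogue and a Season 37 story entry.
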ConversationDatabase /NPCLibrary/NPCs/RoseDepth/NPCLibrary_Conversation_RoseDepth
entry_points [{tag=Conversation.Entry.RoseDepth, destinations [#1]}]
nodes [
 ConversationEntryPointNode_7 {
  id=#1
  class=ConversationEntryPointNode
  outputs [#2]
  properties {entryTag {tagName=Conversation.Entry.RoseDepth}}
 }
 Speech_C_4 {
  id=#2
  class=Speech_C
  outputs [#3, #4, #5, #6]
  properties {
   nodeName=""
   generalConfig {
    defaultMessage=Why are you wasting time speaking to me instead of fighting The Swarm?
    contextualMessages [
     {
      message=I am The Night Rose. What do you want?
      contextRequirements [{requirement {class=FortControllerRequirement_IsFirstEverConversationWithNPC}}]
      priority {curve {rowName=Tandem.Speech.Priority.FirstEverInteraction}}
     }
     {
      message=Why are you wasting time speaking to me instead of fighting The Swarm?
      contextRequirements [{requirement {class=FortControllerRequirement_CheckNPCMatchConversationCount, properties {nPCUniqueIdTag {tagName=AISpawnerData.Type.Tandem.IronLilac}}}}]
      priority {curve {rowName=None}}
     }
     {
      message=There is a plan. And you are not privy to it.
      contextRequirements [{requirement {class=FortControllerRequirement_CheckNPCMatchConversationCount, properties {nPCUniqueIdTag {tagName=AISpawnerData.Type.Tandem.IronLilac}}}}]
      priority {curve {rowName=None}}
     }
     {
      message=We will defeat the creature that rules the Spirit Realm... but not easily.
      contextRequirements [{requirement {class=FortControllerRequirement_CheckNPCMatchConversationCount, properties {nPCUniqueIdTag {tagName=AISpawnerData.Type.Tandem.IronLilac}}}}]
      priority {curve {rowName=None}}
     }
     {
      message=I hope this plan of yours works. Otherwise... this Island is doomed.
      contextRequirements [{requirement {
         class=FortControllerRequirement_HasCID
         properties {allowedCIds [
           &Character_IronLilac_Sly
           &Character_BrightDisk
           &Character_VoidRedemption_Rebel
           &Character_AuraCop_Bullet
           &Character_UndergroundRebel_Fashion
           &CID_989_Athena_Commando_M_ProgressiveJonesy
          ]}
        }}]
      priority {value=2, curve {curveTable=null, rowName=None}}
     }
     {
      message="You'll have to use everything I taught you, now. Make me proud."
      contextRequirements [{requirement {
         class=FortControllerRequirement_HasCID
         properties {allowedCIds [
           /BRCosmetics/Athena/Items/Cosmetics/Characters/Character_BentBaton
           /BRCosmetics/Athena/Items/Cosmetics/Characters/Character_LazyLizzChip
           /BRCosmetics/Athena/Items/Cosmetics/Characters/Character_SpatialTravelWatt
           /BRCosmetics/Athena/Items/Cosmetics/Characters/Character_AshenMagnusSlant
           /BRCosmetics/Athena/Items/Cosmetics/Characters/Character_PolarityWinnField
           /BRCosmetics/Athena/Items/Cosmetics/Characters/Character_MistRavenPeck
          ]}
        }}]
      priority {value=2, curve {curveTable=null, rowName=None}}
     }
    ]
   }
  }
 }
 ConversationLinkNode_3 {
  id=#6
  class=ConversationLinkNode
  properties {remoteEntryTag {tagName=Conversation.Entry.Season37.Story.P03.Q01.A}}
 }
 ConversationLinkNode_2 {id=#5, class=ConversationLinkNode, properties {remoteEntryTag {tagName=Conversation.Entry.RoseDepth.Dialogue}}}
 ConversationLinkNode_0 {id=#4, class=ConversationLinkNode, properties {remoteEntryTag {tagName=Conversation.Entry.BasicServices}}}
 ConversationLinkNode_1 {id=#3, class=ConversationLinkNode, properties {remoteEntryTag {tagName=Conversation.Entry.NPCLibrary.Quests}}}
]
