ConversationDatabase /NPCLibrary/NPCs/RoseDepth/NPCLibrary_Conversation_RoseDepth_Dialogue
entry_points [{tag=Conversation.Entry.RoseDepth.Dialogue, destinations [#1]}]
nodes [
 ConversationEntryPointNode_0 {
  id=#1
  class=ConversationEntryPointNode
  outputs [#2]
  properties {entryTag {tagName=Conversation.Entry.RoseDepth.Dialogue}}
 }
 Speech_C_0 {
  id=#2
  class=Speech_C
  outputs [#3, #4, #5]
  properties {nodeName="", softTaskIcon=&T_UI_NPC_question, generalConfig {defaultMessage=What would you like to know?}}
  sub_nodes [ConversationChoiceNode_5 {class=ConversationChoiceNode, properties {defaultChoiceDisplayText=Ask about...}}]
 }
 Speech_C_1 {
  id=#5
  class=Speech_C
  outputs [#6]
  properties {
   nodeName=""
   softTaskIcon=&T_UI_NPC_question
   generalConfig {defaultMessage="It is funny. I earned my freedom, and now this conflict may rip it from my hands."}
  }
  sub_nodes [ConversationChoiceNode_1 {class=ConversationChoiceNode, properties {defaultChoiceDisplayText=The Supernovas}}]
 }
 Speech_C_2 {
  id=#6
  class=Speech_C
  outputs [#7]
  properties {
   nodeName=""
   softTaskIcon=&T_UI_NPC_question
   generalConfig {defaultMessage="If it does, if we lose... I am proud of what I have accomplished."}
  }
 }
 Speech_C_4 {
  id=#7
  class=Speech_C
  outputs [#8]
  properties {
   nodeName=Speech
   softTaskIcon=&T_UI_NPC_question
   generalConfig {defaultMessage="The Supernovas remind me of those I trained generations ago. Brave, strong."}
  }
 }
 Speech_C_9 {
  id=#8
  class=Speech_C
  outputs [#9]
  properties {
   nodeName=Speech
   softTaskIcon=&T_UI_NPC_question
   generalConfig {defaultMessage=I will stand on the battlefield at their side... come what may.}
  }
 }
 ReturnToStart_C_23 {
  id=#9
  class=ReturnToStart_C
  properties {softTaskIcon=&T_UI_NPC_Back}
  sub_nodes [ConversationChoiceNode_32 {class=ConversationChoiceNode, properties {defaultChoiceDisplayText=BACK}}]
 }
 Back_C_0 {
  id=#4
  class=Back_C
  sub_nodes [ConversationChoiceNode_7 {class=ConversationChoiceNode, properties {defaultChoiceDisplayText=BACK}}]
 }
 Speech_C_7 {
  id=#3
  class=Speech_C
  outputs [#10]
  properties {
   nodeName=""
   softTaskIcon=&T_UI_NPC_question
   generalConfig {defaultMessage=I have serious doubts that we're truly ready to face what awaits us in the Spirit Realm.}
  }
  sub_nodes [ConversationChoiceNode_0 {class=ConversationChoiceNode, properties {defaultChoiceDisplayText=The Spirit Realm}}]
 }
 Speech_C_15 {
  id=#10
  class=Speech_C
  outputs [#11]
  properties {
   nodeName=""
   softTaskIcon=&T_UI_NPC_question
   generalConfig {defaultMessage="But I must admit, begrudgingly... Hope and Jones are right."}
  }
 }
 Speech_C_14 {
  id=#11
  class=Speech_C
  outputs [#12]
  properties {
   nodeName=""
   softTaskIcon=&T_UI_NPC_question
   generalConfig {defaultMessage="With the Swarm gone, our window of opportunity to initiate this conflict is closing rapidly."}
  }
 }
 Speech_C_3 {
  id=#12
  class=Speech_C
  outputs [#9]
  properties {
   nodeName=Speech
   softTaskIcon=&T_UI_NPC_question
   generalConfig {defaultMessage="We must act, and soon. And who knows? Perhaps we'll find ourselves with the unlikeliest upper hand."}
  }
 }
]
